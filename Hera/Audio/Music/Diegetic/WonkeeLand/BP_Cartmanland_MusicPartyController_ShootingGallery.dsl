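; Blueprint music controller for the Cartmanland shooting gallery, driving its diegetic WonkeeLand audio by firing trigger parameters on an audio component.
; One event forwards any named trigger parameter to that audio.
; A custom start/stop event picks a StartGame or StopGame trigger from a boolean, so the audio asset itself handles the actual music transitions.
; The construction script only calls its parent.
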(event EventSetTriggerParameter
  (params (ParameterName :type Name :default "None"))
  (Audio|Parameter|ExecuteTriggerParameter (Variables|Default|GetAudio) ParameterName))

(event Custom|StartStopShootingGalleryMusic
  (params (bShouldStartMusic :type Boolean))
  (Audio|Parameter|ExecuteTriggerParameter (Variables|Default|GetAudio) (Utilities|String|SelectName "StartGame" "StopGame" bShouldStartMusic)))

(fn ConstructionScript
  (|CallParentFunction))
