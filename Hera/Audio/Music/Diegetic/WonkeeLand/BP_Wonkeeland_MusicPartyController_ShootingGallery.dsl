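; Blueprint music controller for the WonkeeLand shooting gallery's diegetic party music, extending a parent controller.
; A trigger-parameter event forwards a given name to the Audio component as an execute-trigger parameter.
; A start/stop music event taking a boolean is declared but has no logic here, and the construction script only calls the parent.
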(event EventSetTriggerParameter
  (params (ParameterName :type Name :default "None"))
  (Audio|Parameter|ExecuteTriggerParameter (Variables|Default|GetAudio) ParameterName))

(event Custom|StartStopShootingGalleryMusic
  (params (bShouldStartMusic :type Boolean)))

(fn ConstructionScript
  (|CallParentFunction))
