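# Mono MetaSound source for the WonkeeLand fireworks music: a countdown intro followed by the fireworks music track.
# On play, the countdown wave plays, and its finish triggers the music wave; both are mixed with fixed dB gains.
# The mix feeds the output and a fireworks audio bus, with base submix and submix sends disabled on the source.
# It plays when silent, and the bStartFireworks input is unconnected.
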MetaSoundSource /Hera/Audio/Music/Diegetic/WonkeeLand/MSS_WonkeeLand_Fireworks_Music
inputs [
 bStartFireworks {type=Bool}
 UE.Source.OnPlay {type=Trigger}
]
interfaces [UE.OutputFormat.Mono:1.0, UE.Source:1.1]
outputs ["UE.OutputFormat.Mono.Audio:0" {type=Audio}]
pages [Default {
  input_defaults {UE.Source.OnPlay=null, bStartFireworks=null}
  nodes [
   DecibelstoLinearGain_0 = UE.Decibels to Linear Gain:Float() {also {defaults {Decibels=3}}}
   DecibelstoLinearGain_1 = UE.Decibels to Linear Gain:Float() {also {defaults {Decibels=6}}}
   WavePlayer_0 = UE.Wave Player:Mono(Play=^UE.Source.OnPlay) {also {defaults {Wave Asset=/Hera/Audio/Music/Diegetic/WonkeeLand/Waves/ThemeParkStageEvent_Countdown_Start}, version [1,1]}}
   WavePlayer_1 = UE.Wave Player:Mono(Play=@WavePlayer_0.On Finished) {also {defaults {Wave Asset=/Hera/Audio/Music/Diegetic/WonkeeLand/Waves/MX_S39_WonkeeLand_Fireworks_LP}, version [1,1]}}
   AudioMixerAudioMixerMono2_0 = "AudioMixer.Audio Mixer (Mono, 2)"(
    Gain 0=@DecibelstoLinearGain_0.Linear Gain
    Gain 1=@DecibelstoLinearGain_1.Linear Gain
    In 1=@WavePlayer_0.Out Mono
    In 0=@WavePlayer_1.Out Mono
   )
   AudioBusWriter1_0 = "UE.Audio Bus Writer (1)"(In 0=@AudioMixerAudioMixerMono2_0.Out) {also {defaults {Audio Bus=&/AB_Wonkeeland_Fireworks}}}
   InitVariable.Float()
   InitVariable.Float()
  ]
  links [{dst="@UE.OutputFormat.Mono.Audio:0", src=AudioMixerAudioMixerMono2_0.Out}]
 }]
properties {
 assetImportData=&self:AssetImportData
 bEnableBaseSubmix=false
 bEnableSubmixSends=false
 bLooping=true
 bOverrideConcurrency=true
 duration=10000
 sampleRate=48000
 soundClassObject=/Game/Sounds/FortSoundClassesAndModes/Fort_11_Base_Music.Fort_11_Base_music
 virtualizationMode=PlayWhenSilent
 concurrencyOverrides {maxCount=4}
}
